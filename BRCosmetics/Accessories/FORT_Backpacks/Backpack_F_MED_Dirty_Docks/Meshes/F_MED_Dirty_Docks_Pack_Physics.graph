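# Physics asset for the Dirty Docks backpack cosmetic mesh, defining collision bodies and joints for its dangling hook and straps.
# The pack_MSTR body, a capsule plus box, anchors the hook and two box straps through angle-limited constraints.
# A second hook body chains off the first with a much wider swing, while the straps keep tighter limits.
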PhysicsAsset /BRCosmetics/Accessories/FORT_Backpacks/Backpack_F_MED_Dirty_Docks/Meshes/F_MED_Dirty_Docks_Pack_Physics
bodies {
 Dyn_Hook [{shapeType=Capsule, length=24.4032, radius=0.9325, center=<-2.2583 0 0>, rotation=<rot 90 0 0>}]
 Dyn_Hook2 [{shapeType=Box, extentX=3.0625, extentY=3.5625, extentZ=1, center=<1.6002 0 0>}]
 Dyn_Strap [{shapeType=Box, extentX=5.4548, extentY=2.9328, extentZ=8.3356, center=<2.7359 0 0>}]
 Dyn_Strap2 [{shapeType=Box, extentX=7.9475, extentY=2.1975, extentZ=9.01, center=<4.0932 0 0>}]
 pack_MSTR [
  {shapeType=Capsule, length=2.2188, radius=4.7812, rotation=<rot -0.002 89.9996 0>}
  {shapeType=Box, extentX=7.5, extentY=20.6875, extentZ=4.8125, center=<1.4065 2.5519 6.3821>, rotation=<rot -2.1496 5.1999 22.3983>}
 ]
}
constraints [
 {bone1Name=Dyn_Hook, bone2Name=pack_MSTR, swing1LimitDegrees=12, swing2LimitDegrees=30, twistLimitDegrees=5}
 {bone1Name=Dyn_Strap2, bone2Name=pack_MSTR, swing1LimitDegrees=15, swing2LimitDegrees=5, twistLimitDegrees=5}
 {bone1Name=Dyn_Strap, bone2Name=pack_MSTR, swing1LimitDegrees=15, swing2LimitDegrees=5, twistLimitDegrees=5}
 {bone1Name=Dyn_Hook2, bone2Name=Dyn_Hook, swing1LimitDegrees=20, swing2LimitDegrees=80, twistLimitDegrees=5}
]
